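# Material function for the BlastBerry map landscape that outputs a camera distance fade factor, named for terrain use.
# It measures the distance from world position to the camera, subtracts the CameraBlendOffset parameter, then divides by CameraBlendDistance.
# The clamped result rises linearly from zero at the offset distance to one across the blend distance, with both values exposed as scalar parameters.
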MaterialFunction /BlastBerryMapContent/Game/Athena/Environments/Landscape/Material/MF_TerrainDistanceFade
graph {nodes [
  n1 = Subtract(A=Distance(A=WorldPosition().XYZ, B=CameraPositionWS()), B=ScalarParameter() {defaultValue=2000, parameterName=CameraBlendOffset})
  n2 = ScalarParameter() {defaultValue=55000, parameterName=CameraBlendDistance}
  FunctionOutput(None=Clamp(None=Divide(A=@n1, B=@n2)))
 ]}
